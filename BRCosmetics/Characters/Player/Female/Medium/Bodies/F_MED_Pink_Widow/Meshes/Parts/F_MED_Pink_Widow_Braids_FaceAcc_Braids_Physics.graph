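# Physics asset for the Pink Widow female character's braids face accessory, defining capsule bodies and joints for left and right braid physics.
# Four-bone braid chains hang from head-linked masters, while two-bone end chains hang from neck_01-linked masters, with swing limits and very tight twist.
# Free-rotation joints also tie each end master to its braid's last bone, bridging the chains.
# Head, neck, clavicle and upper-spine capsules supply body shapes around the braids.
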PhysicsAsset /BRCosmetics/Characters/Player/Female/Medium/Bodies/F_MED_Pink_Widow/Meshes/Parts/F_MED_Pink_Widow_Braids_FaceAcc_Braids_Physics
bodies {
 L_braidend_master [{shapeType=Capsule, length=1, radius=1, shapeName=L_braidend_master_capsule}]
 R_braidend_master [{shapeType=Capsule, length=1, radius=1, shapeName=R_braidend_master_capsule, center=<0.0682 0 0>}]
 clavicle_l [{
   shapeType=Capsule
   length=23
   radius=5.75
   shapeName=clavicle_r_capsule
   center=<9.179 -1.3492 -0.3282>
   rotation=<rot 0.104 -79.707 -100.2929>
  }]
 clavicle_r [{
   shapeType=Capsule
   length=23
   radius=5.75
   shapeName=clavicle_r_capsule
   center=<-10.4455 1.153 -0.1627>
   rotation=<rot 0.104 -79.707 -100.2929>
  }]
 dyn_L_braid_1_jnt [{shapeType=Capsule, length=2.4549, radius=2.5064, shapeName=dyn_R_braid_1_jnt_capsule, center=<2.8689 0.2432 -0.258>, rotation=<rot 90 0 0>}]
 dyn_L_braid_2_jnt [{shapeType=Capsule, length=2.6879, radius=2.7403, shapeName=dyn_R_braid_2_jnt_capsule, center=<3.9625 0 0>, rotation=<rot 90 0 0>}]
 dyn_L_braid_3_jnt [{shapeType=Capsule, length=2.4722, radius=2.6996, shapeName=dyn_R_braid_3_jnt_capsule, center=<3.5871 0 0>, rotation=<rot 90 0 0>}]
 dyn_L_braid_4_jnt [{shapeType=Capsule, length=0, radius=5.375, shapeName=dyn_R_braid_4_jnt_capsule, center=<4.01 -1.7091 0>, rotation=<rot 90 0 0>}]
 dyn_L_braidend_1 [{shapeType=Capsule, length=1.728, radius=1.6894, shapeName=dyn_R_braidend_1_capsule, center=<2.586 0.7655 0>, rotation=<rot 90 0 0>}]
 dyn_L_braidend_2 [{
   shapeType=Capsule
   length=5.0297
   radius=3.7675
   shapeName=dyn_R_braidend_2_capsule
   center=<6.5431 0.6178 0>
   rotation=<rot 90 -176.4237 -166.4239>
  }]
 dyn_R_braid_1_jnt [{
   shapeType=Capsule
   length=2.4549
   radius=2.5064
   shapeName=dyn_R_braid_1_jnt_capsule
   center=<-2.5869 -0.5744 -0.0854>
   rotation=<rot 80.0001 -0.0004 -20.0007>
  }]
 dyn_R_braid_2_jnt [{shapeType=Capsule, length=2.6879, radius=2.7403, shapeName=dyn_R_braid_2_jnt_capsule, center=<-2.905 0 0>, rotation=<rot 90 0 0>}]
 dyn_R_braid_3_jnt [{shapeType=Capsule, length=8.4722, radius=2.6996, shapeName=dyn_R_braid_3_jnt_capsule, center=<-6.9341 0 0>, rotation=<rot 90 0 0>}]
 dyn_R_braid_4_jnt [{shapeType=Capsule, length=2.75, radius=3, shapeName=dyn_R_braid_4_jnt_capsule, center=<-3.2582 -3.1446 0>, rotation=<rot 90 0 -10.0001>}]
 dyn_R_braidend_1 [{shapeType=Capsule, length=1.728, radius=1.6894, shapeName=dyn_R_braidend_1_capsule, center=<-1.7774 -1.6788 0>, rotation=<rot 90 0 0>}]
 dyn_R_braidend_2 [{
   shapeType=Capsule
   length=5.0297
   radius=3.7675
   shapeName=dyn_R_braidend_2_capsule
   center=<-6.5361 -1.6114 0>
   rotation=<rot 90 -176.4237 -166.4239>
  }]
 head [{shapeType=Capsule, length=2, radius=7.75, shapeName=head_capsule, center=<5.546 -1.2017 0>, rotation=<rot -90 180 179.9997>}]
 master_L_braid [{shapeType=Capsule, length=1, radius=1, shapeName=master_L_braid_capsule}]
 master_R_braid [{shapeType=Capsule, length=1, radius=1, shapeName=master_R_braid_capsule}]
 master_braidends_sim_jnt [{shapeType=Capsule, length=1, radius=1, shapeName=master_braidends_sim_jnt_capsule}]
 master_braids_sim_jnt [{shapeType=Capsule, length=0.505, radius=0.505, shapeName=master_braids_sim_jnt_capsule, rotation=<rot 0 -90 0>}]
 neck_01 [{shapeType=Capsule, length=3, radius=3, shapeName=neck_01_capsule, rotation=<rot -90 -178.7342 178.734>}]
 neck_02 [{shapeType=Capsule, length=3, radius=3, shapeName=neck_02_capsule, rotation=<rot -90 -90 89.9998>}]
 spine_05 [{shapeType=Capsule, length=8, radius=8, shapeName=spine_05_capsule, center=<-1.7712 1.1718 0>, rotation=<rot 0 -90 0>}]
}
constraints [
 {
  bone1Name=master_braids_sim_jnt
  bone2Name=head
  swing1LimitDegrees=45
  swing1Motion=Free
  swing2LimitDegrees=45
  swing2Motion=Free
  twistLimitDegrees=45
  twistMotion=Free
 }
 {
  bone1Name=master_L_braid
  bone2Name=master_braids_sim_jnt
  swing1LimitDegrees=45
  swing1Motion=Free
  swing2LimitDegrees=45
  swing2Motion=Free
  twistLimitDegrees=45
  twistMotion=Free
 }
 {
  bone1Name=master_R_braid
  bone2Name=master_braids_sim_jnt
  swing1LimitDegrees=45
  swing1Motion=Free
  swing2LimitDegrees=45
  swing2Motion=Free
  twistLimitDegrees=45
  twistMotion=Free
 }
 {bone1Name=dyn_R_braid_1_jnt, bone2Name=master_R_braid, swing1LimitDegrees=25, swing2LimitDegrees=25, twistLimitDegrees=2}
 {bone1Name=dyn_R_braid_2_jnt, bone2Name=dyn_R_braid_1_jnt, swing1LimitDegrees=25, swing2LimitDegrees=25, twistLimitDegrees=2}
 {bone1Name=dyn_R_braid_3_jnt, bone2Name=dyn_R_braid_2_jnt, swing1LimitDegrees=25, swing2LimitDegrees=25, twistLimitDegrees=2}
 {bone1Name=dyn_R_braid_4_jnt, bone2Name=dyn_R_braid_3_jnt, swing1LimitDegrees=25, swing2LimitDegrees=25, twistLimitDegrees=5}
 {
  bone1Name=master_braidends_sim_jnt
  bone2Name=neck_01
  swing1LimitDegrees=45
  swing1Motion=Free
  swing2LimitDegrees=45
  swing2Motion=Free
  twistLimitDegrees=45
  twistMotion=Free
 }
 {
  bone1Name=L_braidend_master
  bone2Name=master_braidends_sim_jnt
  swing1LimitDegrees=45
  swing1Motion=Free
  swing2LimitDegrees=45
  swing2Motion=Free
  twistLimitDegrees=45
  twistMotion=Free
 }
 {bone1Name=dyn_L_braidend_1, bone2Name=L_braidend_master, swing1LimitDegrees=60, swing2LimitDegrees=50, twistLimitDegrees=1}
 {bone1Name=dyn_L_braidend_2, bone2Name=dyn_L_braidend_1, swing1LimitDegrees=35, swing2LimitDegrees=25, twistLimitDegrees=1}
 {
  bone1Name=R_braidend_master
  bone2Name=master_braidends_sim_jnt
  swing1LimitDegrees=45
  swing1Motion=Free
  swing2LimitDegrees=45
  swing2Motion=Free
  twistLimitDegrees=45
  twistMotion=Free
 }
 {bone1Name=dyn_R_braidend_1, bone2Name=R_braidend_master, swing1LimitDegrees=35, swing2LimitDegrees=35, twistLimitDegrees=1}
 {bone1Name=dyn_R_braidend_2, bone2Name=dyn_R_braidend_1, swing1LimitDegrees=35, swing2LimitDegrees=25, twistLimitDegrees=1}
 {
  bone1Name=R_braidend_master
  bone2Name=dyn_R_braid_4_jnt
  swing1LimitDegrees=45
  swing1Motion=Free
  swing2LimitDegrees=45
  swing2Motion=Free
  twistLimitDegrees=45
  twistMotion=Free
 }
 {bone1Name=dyn_L_braid_1_jnt, bone2Name=master_L_braid, swing1LimitDegrees=25, swing2LimitDegrees=25, twistLimitDegrees=2}
 {bone1Name=dyn_L_braid_2_jnt, bone2Name=dyn_L_braid_1_jnt, swing1LimitDegrees=25, swing2LimitDegrees=25, twistLimitDegrees=2}
 {bone1Name=dyn_L_braid_3_jnt, bone2Name=dyn_L_braid_2_jnt, swing1LimitDegrees=25, swing2LimitDegrees=25, twistLimitDegrees=2}
 {bone1Name=dyn_L_braid_4_jnt, bone2Name=dyn_L_braid_3_jnt, swing1LimitDegrees=25, swing2LimitDegrees=25, twistLimitDegrees=5}
 {
  bone1Name=L_braidend_master
  bone2Name=dyn_L_braid_4_jnt
  swing1LimitDegrees=45
  swing1Motion=Free
  swing2LimitDegrees=45
  swing2Motion=Free
  twistLimitDegrees=45
  twistMotion=Free
 }
]
