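# Control Rig for the Pink Widow female braids face-accessory part, procedurally posing both braids and their end bones from body motion.
# Head rotation weights offsets on the braid-end aim bones, which the mid-braid joints then aim at.
# Smoothed, remapped head-to-spine rotation and curve-shaped spherical pose readers on clavicle, neck and head drive additive braid corrections.
# A right upper-arm correction chain is present but unconnected to the execution entry.
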ControlRigBlueprint /BRCosmetics/Characters/Player/Female/Medium/Bodies/F_MED_Pink_Widow/Meshes/Parts/F_MED_Pink_Widow_Braids_FaceAcc_CtrlRig
graphs [
 RigVMFunctionLibrary {links [], nodes []}
 RigVMModel {nodes [
   n1 = &RigUnit_GetRelativeTransformForItem(Child=(Type=Bone,Name="upperarm_r"), Parent=(Type=Bone,Name="spine_05"))
   n2 = /Script/RigVM.RigVMFunction_AccumulateFloatLerp(
    Blend=10
    InitialValue=0
    bIntegrateDeltaTime=true
    TargetValue=&RigVMFunction_MathFloatRemap(
     SourceMaximum=0
     SourceMinimum=0
     TargetMinimum=0
     bClamp=true
     Value=&RigVMFunction_MathQuaternionToEuler(RotationOrder=YZX, Value=@n1.RelativeTransform.Rotation).Result.Y
    ).Result
   ) {is=struct RigVMUnitNode, template="AccumulateLerp::Execute(in TargetValue,in InitialValue,in Blend,in bIntegrateDeltaTime,out Result)"}
   &RigUnit_ModifyTransforms(
    ExecutePin=()
    ItemToModify=((Item=(Type=Bone,Name="dyn_R_braid_3_jnt"),Transform=(Rotation=(X=0,Y=0,Z=0,W=1),Translation=(X=0,Y=0,Z=0),Scale3D=(X=1,Y=1,Z=1))))
    Weight=@n2.Result
    ItemToModify.0.Transform.Rotation=&RigVMFunction_MathQuaternionFromEuler(Euler=(X=0,Y=50,Z=0)).Result
   )
   > &RigUnit_ModifyTransforms(
    ItemToModify=((Item=(Type=Bone,Name="R_braidend_master"),Transform=(Rotation=(X=0,Y=0,Z=0,W=1),Translation=(X=0,Y=-2,Z=0),Scale3D=(X=1,Y=1,Z=1))))
    Weight=@n2.Result
    ItemToModify.0.Transform.Rotation=&RigVMFunction_MathQuaternionFromEuler(Euler=(X=0,Y=20,Z=0)).Result
   )
   RigUnit_BeginExecution = &RigUnit_BeginExecution()
   n3 = &RigUnit_GetTransform(Item=(Type=Bone,Name="head"), Space=LocalSpace)
   &RigUnit_ModifyTransforms(
    ItemToModify=((Item=(Type=Bone,Name="l_braidend_aim"),Transform=(Rotation=(X=0,Y=0,Z=0,W=1),Translation=(X=-20,Y=-65,Z=0),Scale3D=(X=1,Y=1,Z=1))))
    Weight=&RigVMFunction_MathFloatRemap(
     SourceMaximum=-35
     SourceMinimum=-10
     TargetMaximum=0.2
     TargetMinimum=1
     bClamp=true
     Value=&RigVMFunction_MathQuaternionToEuler(RotationOrder=ZYX, Value=@n3.Transform.Rotation).Result.Z
    ).Result
    ExecutePin=@RigUnit_BeginExecution.ExecutePin
   )
   n4 = &RigUnit_GetTransform(Item=(Type=Bone,Name="head"), Space=LocalSpace)
   > &RigUnit_ModifyTransforms(
    ItemToModify=((Item=(Type=Bone,Name="r_braidend_aim"),Transform=(Rotation=(X=0,Y=0,Z=0,W=1),Translation=(X=25,Y=160,Z=0),Scale3D=(X=1,Y=1,Z=1))))
    WeightMinimum=0
    Weight=&RigVMFunction_MathFloatRemap(
     SourceMaximum=-35
     SourceMinimum=-10
     TargetMaximum=0.03
     TargetMinimum=0.3
     bClamp=true
     Value=&RigVMFunction_MathQuaternionToEuler(RotationOrder=ZYX, Value=@n4.Transform.Rotation).Result.Z
    ).Result
   )
   > /Script/ControlRig.RigUnit_AimItem(
    DebugSettings=(Scale=10,WorldOffset=(Rotation=(X=0,Y=0,Z=0,W=1),Translation=(X=0,Y=0,Z=0),Scale3D=(X=1,Y=1,Z=1)))
    Item=(Type=Bone,Name="dyn_L_braid_2_jnt")
    Primary=(Weight=1,Axis=(X=1,Y=0,Z=0),Target=(X=1,Y=0,Z=0),Kind=Location,Space=(Type=Bone))
    Secondary=(Weight=1,Axis=(X=0,Y=1,Z=0),Target=(X=1,Y=0,Z=0),Kind=Location,Space=(Type=None))
    Weight=1
    Primary.Target=&RigUnit_GetTransform(Item=(Type=Bone,Name="l_braidend_aim")).Transform.Translation
    Secondary.Target=&RigUnit_GetTransform(Item=(Type=Bone,Name="master_L_braid")).Transform.Translation
   ) {is=struct RigVMUnitNode}
   > &RigUnit_AimItem(
    DebugSettings=(Scale=10,WorldOffset=(Rotation=(X=0,Y=0,Z=0,W=1),Translation=(X=0,Y=0,Z=0),Scale3D=(X=1,Y=1,Z=1)))
    Item=(Type=Bone,Name="dyn_R_braid_2_jnt")
    Primary=(Weight=1,Axis=(X=-1,Y=0,Z=0),Target=(X=1,Y=0,Z=0),Kind=Location,Space=(Type=Bone))
    Secondary=(Weight=1,Axis=(X=0,Y=-1,Z=0),Target=(X=1,Y=0,Z=0),Kind=Location,Space=(Type=None))
    Weight=1
    Secondary.Target=&RigUnit_GetTransform(Item=(Type=Bone,Name="master_R_braid")).Transform.Translation
    Primary.Target=&RigUnit_GetTransform(Item=(Type=Bone,Name="r_braidend_aim")).Transform.Translation
   )
   n5 = &RigUnit_GetRelativeTransformForItem(Child=(Type=Bone,Name="head"), Parent=(Type=Bone,Name="spine_05"))
   n6 = &RigVMFunction_AccumulateFloatLerp(
    Blend=10
    InitialValue=0
    bIntegrateDeltaTime=true
    TargetValue=&RigVMFunction_MathFloatRemap(
     SourceMaximum=-35
     SourceMinimum=-10
     TargetMinimum=0
     bClamp=true
     Value=&RigVMFunction_MathQuaternionToEuler(RotationOrder=XZY, Value=@n5.RelativeTransform.Rotation).Result.Z
    ).Result
   ) {template="AccumulateLerp::Execute(in TargetValue,in InitialValue,in Blend,in bIntegrateDeltaTime,out Result)"}
   > &RigUnit_ModifyTransforms(
    ItemToModify=((Item=(Type=Bone,Name="L_braidend_master"),Transform=(Rotation=(X=0,Y=0,Z=0,W=1),Translation=(X=-3,Y=-0.5,Z=0),Scale3D=(X=1,Y=1,Z=1))))
    Weight=@n6.Result
    ItemToModify.0.Transform.Rotation=&RigVMFunction_MathQuaternionFromEuler(Euler=(X=-20,Y=0,Z=0)).Result
   )
   > &RigUnit_ModifyTransforms(
    ItemToModify=((Item=(Type=Bone,Name="R_braidend_master"),Transform=(Rotation=(X=0,Y=0,Z=0,W=1),Translation=(X=5,Y=0,Z=0),Scale3D=(X=1,Y=1,Z=1))))
    Weight=@n6.Result
   )
   n7 = &RigUnit_GetRelativeTransformForItem(Child=(Type=Bone,Name="head"), Parent=(Type=Bone,Name="spine_05"))
   n8 = &RigVMFunction_AccumulateFloatLerp(
    Blend=10
    InitialValue=0
    bIntegrateDeltaTime=true
    TargetValue=&RigVMFunction_MathFloatRemap(
     SourceMaximum=40
     SourceMinimum=-8
     TargetMinimum=0
     bClamp=true
     Value=&RigVMFunction_MathQuaternionToEuler(RotationOrder=XZY, Value=@n7.RelativeTransform.Rotation).Result.Z
    ).Result
   ) {template="AccumulateLerp::Execute(in TargetValue,in InitialValue,in Blend,in bIntegrateDeltaTime,out Result)"}
   > &RigUnit_ModifyTransforms(
    ItemToModify=((Item=(Type=Bone,Name="L_braidend_master"),Transform=(Rotation=(X=0,Y=0,Z=0,W=1),Translation=(X=9,Y=0,Z=0),Scale3D=(X=1,Y=1,Z=1))))
    Weight=@n8.Result
   )
   > &RigUnit_ModifyTransforms(
    ItemToModify=((Item=(Type=Bone,Name="R_braidend_master"),Transform=(Rotation=(X=0,Y=0,Z=0,W=1),Translation=(X=-5,Y=0,Z=0),Scale3D=(X=1,Y=1,Z=1))))
    Weight=@n8.Result
   )
   > n9 = &RigUnit_SphericalPoseReader(
    ActiveRegionScaleFactors=(PositiveWidth=1,NegativeWidth=1,PositiveHeight=1,NegativeHeight=1)
    ActiveRegionSize=0.07
    Debug=(bDrawDebug=false,DebugScale=25,DebugSegments=20,DebugThickness=0.25)
    DriverAxis=(X=1,Y=0,Z=0)
    DriverItem=(Type=Bone,Name="clavicle_l")
    FalloffRegionScaleFactors=(PositiveWidth=1,NegativeWidth=1,PositiveHeight=1,NegativeHeight=1)
    FalloffSize=0.18
    OptionalParentItem=(Type=Bone)
    RotationOffset=(X=5,Y=-40,Z=0)
   )
   n10 = &RigVMFunction_AnimEvalRichCurve(
    Curve=(EditorCurveData=(Keys=((InterpMode=RCIM_Cubic,TangentMode=RCTM_Break,Time=0.0003,Value=-0.004,LeaveTangent=3.1099),(InterpMode=RCIM_Cubic,TangentMode=RCTM_Break,Time=1.0016,Value=0.8563,ArriveTangent=-0.6313))))
    Value=@n9.OutputParam
   )
   > &RigUnit_ModifyTransforms(
    ItemToModify=((Item=(Type=Bone),Transform=(Rotation=(X=0,Y=0,Z=0,W=1),Translation=(X=0,Y=0,Z=0),Scale3D=(X=1,Y=1,Z=1))),(Item=(Type=Bone,Name="R_braidend_master"),Transform=(Rotation=(X=0,Y=0,Z=0,W=1),Translation=(X=0,Y=0,Z=1),Scale3D=(X=1,Y=1,Z=1))))
    WeightMinimum=0
    Weight=@n10.Result
    ItemToModify.1.Transform.Rotation=&RigVMFunction_MathQuaternionFromEuler(Euler=(X=20,Y=30,Z=45)).Result
   )
   > &RigUnit_ModifyTransforms(
    ItemToModify=((Item=(Type=Bone),Transform=(Rotation=(X=0,Y=0,Z=0,W=1),Translation=(X=0,Y=0,Z=0),Scale3D=(X=1,Y=1,Z=1))),(Item=(Type=Bone,Name="L_braidend_master"),Transform=(Rotation=(X=0,Y=0,Z=0,W=1),Translation=(X=3,Y=0,Z=0),Scale3D=(X=1,Y=1,Z=1))))
    WeightMinimum=0
    Weight=@n10.Result
   )
   > &RigUnit_ModifyTransforms(
    ItemToModify=((Item=(Type=Bone),Transform=(Rotation=(X=0,Y=0,Z=0,W=1),Translation=(X=0,Y=0,Z=0),Scale3D=(X=1,Y=1,Z=1))),(Item=(Type=Bone,Name="master_R_braid"),Transform=(Rotation=(X=0,Y=0,Z=0,W=1),Translation=(X=0,Y=0,Z=2),Scale3D=(X=1,Y=1,Z=1))))
    WeightMinimum=0
    Weight=@n10.Result
    ItemToModify.1.Transform.Rotation=&RigVMFunction_MathQuaternionFromEuler(Euler=(X=0,Y=20,Z=0)).Result
   )
   n11 = &RigUnit_SphericalPoseReader(
    ActiveRegionScaleFactors=(PositiveWidth=1,NegativeWidth=1,PositiveHeight=1,NegativeHeight=1)
    ActiveRegionSize=0.06
    Debug=(bDrawDebug=false,DebugScale=25,DebugSegments=20,DebugThickness=0.25)
    DriverAxis=(X=1,Y=1,Z=0)
    DriverItem=(Type=Bone,Name="head")
    ExecutePin=()
    FalloffRegionScaleFactors=(PositiveWidth=1,NegativeWidth=1,PositiveHeight=1,NegativeHeight=1)
    FalloffSize=0.09
    OptionalParentItem=(Type=Bone)
    RotationOffset=(X=94,Y=0,Z=-27)
   )
   n12 = &RigVMFunction_AnimEvalRichCurve(
    Curve=(EditorCurveData=(Keys=((InterpMode=RCIM_Cubic,TangentMode=RCTM_User,ArriveTangent=3.6156,LeaveTangent=3.6156),(InterpMode=RCIM_Cubic,TangentMode=RCTM_User,Time=1,Value=1,ArriveTangent=2.8364,LeaveTangent=2.8364))))
    Value=@n11.OutputParam
   )
   > &RigUnit_ModifyTransforms(
    ItemToModify=((Item=(Type=Bone,Name="L_braidend_master"),Transform=(Rotation=(X=0,Y=0,Z=0,W=1),Translation=(X=3,Y=1.5,Z=0),Scale3D=(X=1,Y=1,Z=1))))
    Weight=@n12.Result
    ItemToModify.0.Transform.Rotation=&RigVMFunction_MathQuaternionFromEuler(Euler=(X=0,Y=0,Z=-20)).Result
   )
   n13 = &RigUnit_SphericalPoseReader(
    ActiveRegionScaleFactors=(PositiveWidth=1,NegativeWidth=1,PositiveHeight=1,NegativeHeight=1)
    ActiveRegionSize=0.1
    Debug=(bDrawDebug=True,DebugScale=25,DebugSegments=20,DebugThickness=0.25)
    DriverAxis=(X=1,Y=0,Z=0)
    DriverItem=(Type=Bone,Name="neck_01")
    ExecutePin=()
    FalloffRegionScaleFactors=(PositiveWidth=1,NegativeWidth=1,PositiveHeight=1,NegativeHeight=1)
    FalloffSize=0.12
    OptionalParentItem=(Type=Bone)
    RotationOffset=(X=85,Y=0,Z=-110)
   )
   n14 = &RigVMFunction_AnimEvalRichCurve(
    Curve=(EditorCurveData=(Keys=((InterpMode=RCIM_Cubic,TangentMode=RCTM_User,ArriveTangent=2.8035,LeaveTangent=2.8035),(InterpMode=RCIM_Cubic,TangentMode=RCTM_User,Time=1,Value=1,ArriveTangent=2.4396,LeaveTangent=2.4396))))
    Value=@n13.OutputParam
   )
   > &RigUnit_ModifyTransforms(
    ItemToModify=((Item=(Type=Bone,Name="L_braidend_master"),Transform=(Rotation=(X=0,Y=0,Z=0,W=1),Translation=(X=-3,Y=0,Z=0),Scale3D=(X=1,Y=1,Z=1))))
    Weight=@n14.Result
    ItemToModify.0.Transform.Rotation=&RigVMFunction_MathQuaternionFromEuler(Euler=(X=0,Y=0,Z=0)).Result
   )
   n15 = &RigUnit_SphericalPoseReader(
    ActiveRegionScaleFactors=(PositiveWidth=1,NegativeWidth=1,PositiveHeight=1,NegativeHeight=1)
    ActiveRegionSize=0.06
    Debug=(bDrawDebug=True,DebugScale=25,DebugSegments=20,DebugThickness=0.25)
    DriverAxis=(X=1,Y=1,Z=0)
    DriverItem=(Type=Bone,Name="head")
    ExecutePin=()
    FalloffRegionScaleFactors=(PositiveWidth=1,NegativeWidth=1,PositiveHeight=1,NegativeHeight=1)
    FalloffSize=0.08
    OptionalParentItem=(Type=Bone)
    RotationOffset=(X=94,Y=0,Z=-27)
   )
   n16 = &RigVMFunction_AnimEvalRichCurve(
    Curve=(EditorCurveData=(Keys=((InterpMode=RCIM_Cubic,TangentMode=RCTM_User,ArriveTangent=3.6156,LeaveTangent=3.6156),(InterpMode=RCIM_Cubic,TangentMode=RCTM_User,Time=1,Value=1,ArriveTangent=2.8364,LeaveTangent=2.8364))))
    Value=@n15.OutputParam
   )
   > &RigUnit_ModifyTransforms(
    ItemToModify=((Item=(Type=Bone,Name="R_braidend_master"),Transform=(Rotation=(X=0,Y=0,Z=0,W=1),Translation=(X=-1,Y=2,Z=1),Scale3D=(X=1,Y=1,Z=1))))
    Weight=@n16.Result
    ItemToModify.0.Transform.Rotation=&RigVMFunction_MathQuaternionFromEuler(Euler=(X=0,Y=0,Z=10)).Result
   )
   n17 = &RigUnit_SphericalPoseReader(
    ActiveRegionScaleFactors=(PositiveWidth=1,NegativeWidth=1,PositiveHeight=1,NegativeHeight=1)
    ActiveRegionSize=0.08
    Debug=(bDrawDebug=True,DebugScale=25,DebugSegments=20,DebugThickness=0.25)
    DriverAxis=(X=1,Y=1,Z=0)
    DriverItem=(Type=Bone,Name="head")
    ExecutePin=()
    FalloffRegionScaleFactors=(PositiveWidth=1,NegativeWidth=1,PositiveHeight=1,NegativeHeight=1)
    FalloffSize=0.09
    OptionalParentItem=(Type=Bone)
    RotationOffset=(X=94,Y=0,Z=-27)
   )
   n18 = &RigVMFunction_AnimEvalRichCurve(
    Curve=(EditorCurveData=(Keys=((InterpMode=RCIM_Cubic,TangentMode=RCTM_User,ArriveTangent=2.8035,LeaveTangent=2.8035),(InterpMode=RCIM_Cubic,TangentMode=RCTM_User,Time=1,Value=1,ArriveTangent=2.4396,LeaveTangent=2.4396))))
    Value=@n17.OutputParam
   )
   > &RigUnit_ModifyTransforms(
    ItemToModify=((Item=(Type=Bone,Name="R_braidend_master"),Transform=(Rotation=(X=0,Y=0,Z=0,W=1),Translation=(X=-1,Y=2,Z=1),Scale3D=(X=1,Y=1,Z=1))))
    WeightMinimum=0
    Weight=@n18.Result
    ItemToModify.0.Transform.Rotation=&RigVMFunction_MathQuaternionFromEuler(Euler=(X=0,Y=0,Z=-5)).Result
   )
   &RigVMFunction_MathQuaternionFromEuler(Euler=(X=0,Y=0,Z=0))
   &RigVMFunction_MathQuaternionFromEuler(Euler=(X=0,Y=0,Z=-20))
   /Script/ControlRig.RigUnit_SequenceExecution(ExecuteContext=()) {is=struct RigVMUnitNode}
  ]}
]
hierarchy [
 root {loc=<0 0 0>, rot=<0 0 0 1>}
 attach {parent=root, loc=<0 0 0>, rot=<0 0 0 1>}
 pelvis {parent=root, loc=<0 -0.5301 98.0571>, rot=<0.0127 0.707 -0.0127 -0.707>}
 spine_01 {parent=pelvis, loc=<6.1208 -0.9328 0>, rot=<0 0 0.0576 -0.9983>}
 spine_02 {parent=spine_01, loc=<6.6022 -0.0153 0>, rot=<0 0 0.0022 -1>}
 spine_03 {parent=spine_02, loc=<6.4919 -0.0105 0>, rot=<0 0 0.0015 -1>}
 spine_04 {parent=spine_03, loc=<6.9596 -0.007 0>, rot=<0 0 0.0009 -1>}
 spine_05 {parent=spine_04, loc=<6.9981 -0.0017 0>, rot=<0 0 -0.0042 -1>}
 clavicle_l {parent=spine_05, loc=<5.4831 -0.1053 -3.2946>, rot=<0.0938 -0.7388 0.0307 -0.6666>}
 upperarm_l {parent=clavicle_l, loc=<9.9486 0 0>, rot=<0.0371 -0.2992 -0.0705 -0.9509>}
 neck_01 {parent=spine_05, loc=<11.3138 0.1529 0>, rot=<0 0 -0.2003 -0.9797>}
 neck_02 {parent=neck_01, loc=<4.5627 0 0>, rot=<0 0 0.0504 -0.9987>}
 head {parent=neck_02, loc=<5.3914 0 0>, rot=<0 0 0.115 -0.9934>}
 master_braids_sim_jnt {parent=head, loc=<11.8239 -2.5575 0>, rot=<0 0 0 1>}
 master_L_braid {parent=master_braids_sim_jnt, loc=<1.3564 -0.5317 -5.1591>, rot=<0.8645 -0.3869 -0.1503 -0.2833>}
 dyn_L_braid_1_jnt {parent=master_L_braid, loc=<3.741 2.472 -0.4816>, rot=<-0.0874 -0.0774 -0.6172 -0.7781>}
 dyn_L_braid_2_jnt {parent=dyn_L_braid_1_jnt, loc=<4.9967 1.7978 -0.3844>, rot=<0.0377 -0.0658 -0.3611 -0.9294>}
 dyn_L_braid_3_jnt {parent=dyn_L_braid_2_jnt, loc=<6.3966 0.6677 -0.4004>, rot=<-0.1106 -0.0178 -0.1168 -0.9868>}
 dyn_L_braid_4_jnt {parent=dyn_L_braid_3_jnt, loc=<5.4173 0.8196 -0.2046>, rot=<0.0215 0.0135 -0.1828 -0.9828>}
 master_R_braid {parent=master_braids_sim_jnt, loc=<1.3569 -0.5317 5.1591>, rot=<-0.3869 -0.8645 0.2833 -0.1503>}
 dyn_R_braid_1_jnt {parent=master_R_braid, loc=<-3.7406 -2.4724 0.4816>, rot=<-0.0874 -0.0774 -0.6172 -0.7781>}
 dyn_R_braid_2_jnt {parent=dyn_R_braid_1_jnt, loc=<-4.9965 -1.7976 0.3844>, rot=<0.0377 -0.0658 -0.3611 -0.9294>}
 dyn_R_braid_3_jnt {parent=dyn_R_braid_2_jnt, loc=<-6.3971 -0.6678 0.4004>, rot=<-0.1106 -0.0178 -0.1168 -0.9868>}
 dyn_R_braid_4_jnt {parent=dyn_R_braid_3_jnt, loc=<-5.4165 -0.8197 0.2047>, rot=<0.0215 0.0135 -0.1828 -0.9828>}
 master_L_earring_jnt {parent=head, loc=<-0.3677 0.4144 -5.989>, rot=<0.0029 -0.7071 -0.0029 -0.7071>}
 dyn_L_earring_jnt {parent=master_L_earring_jnt, loc=<0 0 -0.6028>, rot=<0 -0.7071 0 -0.7071>}
 master_R_earring_jnt {parent=head, loc=<-0.3119 0.4149 5.989>, rot=<0.7071 0.0029 -0.7071 0.0029>}
 dyn_R_earring_jnt {parent=master_R_earring_jnt, loc=<0 0 0.6587>, rot=<0.7071 0 0.7071 0>}
 master_braidends_sim_jnt {parent=neck_01, loc=<0 -2.9032 0>, rot=<0 0 0 1>}
 L_braidend_master {parent=master_braidends_sim_jnt, loc=<0.5337 0.6775 -7.7991>, rot=<-0.6338 -0.7664 -0.0878 -0.0565>}
 dyn_L_braidend_1 {parent=L_braidend_master, loc=<0.9465 -0.0248 0.1218>, rot=<0 0 0 1>}
 dyn_L_braidend_2 {parent=dyn_L_braidend_1, loc=<4.0772 -0.1936 0.386>, rot=<-0.0989 0.0479 0.0846 -0.9903>}
 l_braidend_aim {parent=L_braidend_master, loc=<0 0 0>, rot=<0 0 0 1>}
 R_braidend_master {parent=master_braidends_sim_jnt, loc=<0.5334 0.6776 7.7991>, rot=<0.7664 -0.6338 -0.0565 0.0878>}
 dyn_R_braidend_1 {parent=R_braidend_master, loc=<-0.9463 0.0245 -0.1219>, rot=<0 0 0 -1>}
 dyn_R_braidend_2 {parent=dyn_R_braidend_1, loc=<-4.077 0.1932 -0.3861>, rot=<-0.0989 0.0479 0.0846 -0.9903>}
 r_braidend_aim {parent=R_braidend_master, loc=<0 0 0>, rot=<0 0 0 1>}
 clavicle_r {parent=spine_05, loc=<5.4831 -0.1053 3.2946>, rot=<0.7388 0.0938 -0.6666 -0.0307>}
 upperarm_r {parent=clavicle_r, loc=<-9.9486 0 0>, rot=<0.0371 -0.2992 -0.0705 -0.9509>}
]
