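# Physics asset for the earrings on the Pink Widow female character's braids face-accessory mesh, providing bodies and joints for earring dangle.
# Capsules cover the head, a small master joint per ear and a larger dynamic joint per earring.
# Each side chains dynamic joint to master to head.
# The earring joint swings mainly on one axis, with the other swing and twist nearly locked, while master-to-head rotation is free.
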PhysicsAsset /BRCosmetics/Characters/Player/Female/Medium/Bodies/F_MED_Pink_Widow/Meshes/Parts/F_MED_Pink_Widow_Braids_FaceAcc_Earrings_Physics
bodies {
 dyn_L_earring_jnt [{
   shapeType=Capsule
   length=3.5129
   radius=1.6899
   shapeName=dyn_L_earring_jnt_capsule
   center=<3.4483 0.1352 0.0186>
   rotation=<rot 0 92.0613 -89.5712>
  }]
 dyn_R_earring_jnt [{
   shapeType=Capsule
   length=3.5129
   radius=1.6899
   shapeName=dyn_R_earring_jnt_capsule
   center=<3.5184 0.1156 0.0336>
   rotation=<rot 0 92.0613 -89.5712>
  }]
 head [{shapeType=Capsule, length=5, radius=5, shapeName=head_capsule, center=<0 3.149 0>, rotation=<rot -90 180 179.9997>}]
 master_L_earring_jnt [{
   shapeType=Capsule
   length=0.505
   radius=0.505
   shapeName=master_L_earring_jnt_capsule
   center=<0.0174 0.0072 0.0334>
   rotation=<rot 0.457 -90 -0.1315>
  }]
 master_R_earring_jnt [{
   shapeType=Capsule
   length=0.505
   radius=0.505
   shapeName=master_R_earring_jnt_capsule
   center=<-0.0176 -0.0072 0.0223>
   rotation=<rot 0.4569 -90 -0.1315>
  }]
}
constraints [
 {bone1Name=dyn_R_earring_jnt, bone2Name=master_R_earring_jnt, swing1LimitDegrees=45, swing2LimitDegrees=2, twistLimitDegrees=2}
 {
  bone1Name=master_R_earring_jnt
  bone2Name=head
  swing1LimitDegrees=45
  swing1Motion=Free
  swing2LimitDegrees=45
  swing2Motion=Free
  twistLimitDegrees=45
  twistMotion=Free
 }
 {bone1Name=dyn_L_earring_jnt, bone2Name=master_L_earring_jnt, swing1LimitDegrees=45, swing2LimitDegrees=2, twistLimitDegrees=2}
 {
  bone1Name=master_L_earring_jnt
  bone2Name=head
  swing1LimitDegrees=45
  swing1Motion=Free
  swing2LimitDegrees=45
  swing2Motion=Free
  twistLimitDegrees=45
  twistMotion=Free
 }
]
