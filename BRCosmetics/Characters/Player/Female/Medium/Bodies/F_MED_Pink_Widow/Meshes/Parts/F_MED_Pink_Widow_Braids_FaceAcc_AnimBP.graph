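# Post-process-style AnimBP for Pink Widow's braids and earrings, copying the parent mesh's pose.
# Pose runs through a Control Rig, emote-only left-braid constraint, head-space RigidBody sims for braids then earrings, and a secondary Control Rig.
# Braid-end offsets apply in jump/fall and front end; update event lowers braid sim weight when DBNO or skydiving and suppresses the jump offset.
# A per-bone AnimDynamics braid chain remains in the graph, unconnected to output.
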AnimBlueprint /BRCosmetics/Characters/Player/Female/Medium/Bodies/F_MED_Pink_Widow/Meshes/Parts/F_MED_Pink_Widow_Braids_FaceAcc_AnimBP
target_skeleton=&/F_MED_Pink_Widow_Braids_FaceAcc_Skeleton
graphs [
 AnimGraph {nodes [
   n1 = Anim Dynamics - Bone: dyn_L_braid_1_jnt(Alpha=1, ExternalForce="0,0,0") {
    angularDampingOverride=0.9
    angularSpringConstant=500
    bAngularSpring=true
    bOverrideAngularDamping=true
    bUseGravityOverride=true
    gravityScale=0.35
    simulationSpace=RootRelative
    alphaBoolBlend {blendOption=Linear}
    boundBone {boneName=dyn_L_braid_1_jnt}
    componentAppliedLinearAccClamp=<1200 1200 1200>
    componentLinearAccScale=<1.4 1.4 1.4>
    componentLinearVelScale=<1.4 1.4 1.4>
    gravityOverride {z=-980}
    physicsBodyDefinitions [{
      sphereCollisionRadius=0
      boundBone {boneName=dyn_L_braid_1_jnt}
      boxExtents=<25 25 25>
      constraintSetup {angularLimitsMax=<5 5 5>, angularLimitsMin=<-5 -5 -10>, angularTarget=<1 0 0>}
      localJointOffset=<12 0 0>
     }]
    relativeSpaceBone {boneName=master_braids_sim_jnt}
   }
   n2 = Anim Dynamics - Bone: dyn_L_braid_2_jnt(ComponentPose=@n1, Alpha=1, ExternalForce="0,0,0") {
    angularDampingOverride=0.9
    angularSpringConstant=500
    bAngularSpring=true
    bOverrideAngularDamping=true
    bUseGravityOverride=true
    gravityScale=0.35
    simulationSpace=RootRelative
    alphaBoolBlend {blendOption=Linear}
    boundBone {boneName=dyn_L_braid_2_jnt}
    componentAppliedLinearAccClamp=<800 800 800>
    componentLinearAccScale=<1.4 1.4 1.4>
    componentLinearVelScale=<1.4 1.4 1.4>
    gravityOverride {z=-980}
    physicsBodyDefinitions [{
      sphereCollisionRadius=0
      boundBone {boneName=dyn_L_braid_2_jnt}
      boxExtents=<20 20 20>
      constraintSetup {angularLimitsMax=<5 5 5>, angularLimitsMin=<-5 -5 -10>, angularTarget=<1 0 0>}
      localJointOffset=<10 0 0>
     }]
    relativeSpaceBone {boneName=master_braids_sim_jnt}
   }
   n3 = Anim Dynamics - Bone: dyn_L_braid_3_jnt(ComponentPose=@n2, Alpha=1, ExternalForce="0,0,0") {
    angularDampingOverride=0.9
    angularSpringConstant=500
    bOverrideAngularDamping=true
    bUseGravityOverride=true
    gravityScale=0.35
    simulationSpace=RootRelative
    alphaBoolBlend {blendOption=Linear}
    boundBone {boneName=dyn_L_braid_3_jnt}
    componentAppliedLinearAccClamp=<200 200 200>
    componentLinearAccScale=<1.3 1.3 1.3>
    componentLinearVelScale=<1.5 1.5 1.5>
    gravityOverride {z=-980}
    physicsBodyDefinitions [{
      sphereCollisionRadius=0
      boundBone {boneName=dyn_L_braid_3_jnt}
      boxExtents=<15 15 15>
      constraintSetup {angularLimitsMax=<5 5 5>, angularLimitsMin=<-5 -5 -5>, angularTarget=<1 0 0>}
      localJointOffset=<7 0 0>
     }]
    relativeSpaceBone {boneName=master_braids_sim_jnt}
   }
   n4 = Anim Dynamics - Bone: dyn_L_braid_4_jnt(ComponentPose=@n3, Alpha=1, ExternalForce="0,0,0") {
    angularDampingOverride=0.9
    angularSpringConstant=500
    bOverrideAngularDamping=true
    bUseGravityOverride=true
    gravityScale=0.35
    simulationSpace=RootRelative
    alphaBoolBlend {blendOption=Linear}
    boundBone {boneName=dyn_L_braid_4_jnt}
    componentAppliedLinearAccClamp=<200 200 200>
    componentLinearAccScale=<1.3 1.3 1.3>
    componentLinearVelScale=<1.5 1.5 1.5>
    gravityOverride {z=-980}
    physicsBodyDefinitions [{
      sphereCollisionRadius=0
      boundBone {boneName=dyn_L_braid_4_jnt}
      boxExtents=<15 15 15>
      constraintSetup {angularLimitsMax=<5 5 5>, angularLimitsMin=<-5 -5 -5>, angularTarget=<1 0 0>}
      localJointOffset=<7 0 0>
     }]
    relativeSpaceBone {boneName=master_braids_sim_jnt}
   }
   n5 = Anim Dynamics - Bone: dyn_L_braidend_1(ComponentPose=@n4, Alpha=1, ExternalForce="0,0,0") {
    angularDampingOverride=0.9
    angularSpringConstant=500
    bAngularSpring=true
    bOverrideAngularDamping=true
    bUseGravityOverride=true
    gravityScale=0.35
    simulationSpace=Actor
    alphaBoolBlend {blendOption=Linear}
    boundBone {boneName=dyn_L_braidend_1}
    componentLinearAccScale=<1.2 1.2 1.1>
    componentLinearVelScale=<1.2 1.2 1.1>
    gravityOverride {z=-980}
    physicsBodyDefinitions [{
      sphereCollisionRadius=0
      boundBone {boneName=dyn_L_braidend_1}
      boxExtents=<25 25 25>
      constraintSetup {angularLimitsMax=<5 5 50>, angularLimitsMin=<-5 -5 10>, angularTarget=<1 0 0>}
      localJointOffset=<12 0 0>
     }]
    relativeSpaceBone {boneName=pelvis}
   }
   n6 = Anim Dynamics - Bone: dyn_L_braidend_2(ComponentPose=@n5, Alpha=1, ExternalForce="0,0,0") {
    angularDampingOverride=0.9
    angularSpringConstant=500
    bAngularSpring=true
    bOverrideAngularDamping=true
    bUseGravityOverride=true
    gravityScale=0.35
    simulationSpace=Actor
    alphaBoolBlend {blendOption=Linear}
    boundBone {boneName=dyn_L_braidend_2}
    componentAppliedLinearAccClamp=<1000 1000 1000>
    componentLinearAccScale=<1.2 1.2 1.2>
    componentLinearVelScale=<1.2 1.2 1.2>
    gravityOverride {z=-980}
    physicsBodyDefinitions [{
      sphereCollisionRadius=0
      boundBone {boneName=dyn_L_braidend_2}
      boxExtents=<25 25 25>
      constraintSetup {angularLimitsMax=<50 5 45>, angularLimitsMin=<-50 -30 10>, angularTarget=<1 0 0>}
      localJointOffset=<12 0 0>
     }]
    relativeSpaceBone {boneName=pelvis}
   }
   n7 = Anim Dynamics - Bone: dyn_R_braid_1_jnt(ComponentPose=@n6, Alpha=1, ExternalForce="0,0,0") {
    angularDampingOverride=0.9
    angularSpringConstant=500
    bAngularSpring=true
    bOverrideAngularDamping=true
    bUseGravityOverride=true
    gravityScale=0.35
    simulationSpace=RootRelative
    alphaBoolBlend {blendOption=Linear}
    boundBone {boneName=dyn_R_braid_1_jnt}
    componentAppliedLinearAccClamp=<1200 1200 1200>
    componentLinearAccScale=<1.4 1.4 1.4>
    componentLinearVelScale=<1.4 1.4 1.4>
    gravityOverride {z=-980}
    physicsBodyDefinitions [{
      sphereCollisionRadius=0
      boundBone {boneName=dyn_R_braid_1_jnt}
      boxExtents=<25 25 25>
      constraintSetup {angularLimitsMax=<5 5 5>, angularLimitsMin=<-5 -5 -30>, angularTarget=<1 0 0>}
      localJointOffset=<-12 0 0>
     }]
    relativeSpaceBone {boneName=master_braids_sim_jnt}
   }
   n8 = Anim Dynamics - Bone: dyn_R_braid_2_jnt(ComponentPose=@n7, Alpha=1, ExternalForce="0,0,0") {
    angularDampingOverride=0.9
    angularSpringConstant=500
    bAngularSpring=true
    bOverrideAngularDamping=true
    bUseGravityOverride=true
    gravityScale=0.35
    simulationSpace=RootRelative
    alphaBoolBlend {blendOption=Linear}
    boundBone {boneName=dyn_R_braid_2_jnt}
    componentAppliedLinearAccClamp=<800 800 800>
    componentLinearAccScale=<1.4 1.4 1.4>
    componentLinearVelScale=<1.4 1.4 1.4>
    gravityOverride {z=-980}
    physicsBodyDefinitions [{
      sphereCollisionRadius=0
      boundBone {boneName=dyn_R_braid_2_jnt}
      boxExtents=<20 20 20>
      constraintSetup {angularLimitsMax=<5 5 5>, angularLimitsMin=<-5 -5 -15>, angularTarget=<1 0 0>}
      localJointOffset=<-10 0 0>
     }]
    relativeSpaceBone {boneName=master_braids_sim_jnt}
   }
   n9 = Anim Dynamics - Bone: dyn_R_braid_3_jnt(ComponentPose=@n8, Alpha=1, ExternalForce="0,0,0") {
    angularDampingOverride=0.9
    angularSpringConstant=500
    bOverrideAngularDamping=true
    bUseGravityOverride=true
    simulationSpace=RootRelative
    alphaBoolBlend {blendOption=Linear}
    boundBone {boneName=dyn_R_braid_3_jnt}
    componentAppliedLinearAccClamp=<200 200 200>
    componentLinearAccScale=<1.3 1.3 1.3>
    componentLinearVelScale=<1.5 1.5 1.5>
    gravityOverride {z=-980}
    physicsBodyDefinitions [{
      boundBone {boneName=dyn_R_braid_3_jnt}
      boxExtents=<15 15 15>
      constraintSetup {angularLimitsMax=<5 5 5>, angularLimitsMin=<-5 -5 -5>, angularTarget=<1 0 0>}
      localJointOffset=<-7 0 0>
     }]
    relativeSpaceBone {boneName=master_braids_sim_jnt}
   }
   n10 = Anim Dynamics - Bone: dyn_R_braid_4_jnt(ComponentPose=@n9, Alpha=1, ExternalForce="0,0,0") {
    angularDampingOverride=0.9
    angularSpringConstant=500
    bOverrideAngularDamping=true
    bUseGravityOverride=true
    simulationSpace=RootRelative
    alphaBoolBlend {blendOption=Linear}
    boundBone {boneName=dyn_R_braid_4_jnt}
    componentAppliedLinearAccClamp=<200 200 200>
    componentLinearAccScale=<1.3 1.3 1.3>
    componentLinearVelScale=<1.5 1.5 1.5>
    gravityOverride {z=-980}
    physicsBodyDefinitions [{
      boundBone {boneName=dyn_R_braid_4_jnt}
      boxExtents=<15 15 15>
      constraintSetup {angularLimitsMax=<5 5 5>, angularLimitsMin=<-5 -5 -5>, angularTarget=<1 0 0>}
      localJointOffset=<-7 0 0>
     }]
    relativeSpaceBone {boneName=master_braids_sim_jnt}
   }
   n11 = Anim Dynamics - Bone: dyn_R_braidend_1(ComponentPose=@n10, Alpha=1, ExternalForce="0,0,0") {
    angularDampingOverride=0.9
    angularSpringConstant=500
    bAngularSpring=true
    bOverrideAngularDamping=true
    bUseGravityOverride=true
    gravityScale=0.35
    simulationSpace=Actor
    alphaBoolBlend {blendOption=Linear}
    boundBone {boneName=dyn_R_braidend_1}
    componentLinearAccScale=<1.2 1.2 1.1>
    componentLinearVelScale=<1.2 1.2 1.1>
    gravityOverride {z=-980}
    physicsBodyDefinitions [{
      sphereCollisionRadius=0
      boundBone {boneName=dyn_R_braidend_1}
      boxExtents=<25 25 25>
      constraintSetup {angularLimitsMax=<5 5 50>, angularLimitsMin=<-5 -5 10>, angularTarget=<1 0 0>}
      localJointOffset=<-12 0 0>
     }]
    relativeSpaceBone {boneName=pelvis}
   }
   Anim Dynamics - Bone: dyn_R_braidend_2(ComponentPose=@n11, Alpha=1, ExternalForce="0,0,0") {
    angularDampingOverride=0.9
    angularSpringConstant=500
    bAngularSpring=true
    bOverrideAngularDamping=true
    bUseGravityOverride=true
    gravityScale=0.35
    simulationSpace=Actor
    alphaBoolBlend {blendOption=Linear}
    boundBone {boneName=dyn_R_braidend_2}
    componentAppliedLinearAccClamp=<1000 1000 1000>
    componentLinearAccScale=<1.2 1.2 1.2>
    componentLinearVelScale=<1.2 1.2 1.2>
    gravityOverride {z=-980}
    physicsBodyDefinitions [{
      sphereCollisionRadius=0
      boundBone {boneName=dyn_R_braidend_2}
      boxExtents=<25 25 25>
      constraintSetup {angularLimitsMax=<50 5 75>, angularLimitsMin=<-50 -30 -10>, angularTarget=<1 0 0>}
      localJointOffset=<-12 0 0>
     }]
    relativeSpaceBone {boneName=pelvis}
   }
   n12 = Control Rig(Source=Copy Pose From Mesh() {bUseAttachedParent=true}) {
    alphaBoolBlend {blendOption=Linear}
    controlRigAssetReference {blueprintRigClass=&/F_MED_Pink_Widow_Braids_FaceAcc_CtrlRig.F_MED_Pink_Widow_Braids_FaceAcc_CtrlRig_C}
   }
   n13 = "Constraint - L_braidend_master (PP)"!(
    ComponentPose=Local To Component(LocalPose=@n12)
    bAlphaBoolEnabled:bool=Get bIsPlayingEmote()
    ConstraintWeights_0=0.7
    ConstraintWeights_1:float=0.3
   ) -> (Pose) {
    alphaInputType=Bool
    alphaBoolBlend {blendInTime=0.1, blendOption=Linear, blendOutTime=0.1}
    boneToModify {boneName=L_braidend_master}
    constraintSetup [{targetBone {boneName=clavicle_l}}, {targetBone {boneName=head}}]
    constraintWeights [1,1]
   }
   n14 = RigidBody(ComponentPose=@n13, Alpha=Get Braid alpha() -> (Braid alpha:double)) {
    bOverrideWorldGravity=true
    overridePhysicsAsset=&/F_MED_Pink_Widow_Braids_FaceAcc_Braids_Physics
    simulationSpace=BaseBoneSpace
    alphaBoolBlend {blendOption=Linear}
    baseBoneRef {boneName=head}
    componentAppliedLinearAccClamp=<3000 3000 3000>
    componentLinearAccScale=<0.4 0.3 0.3>
    componentLinearVelScale=<0.4 0.3 0.3>
    overrideWorldGravity {z=-980}
    simSpaceSettings {maxAngularAcceleration=1000, maxAngularVelocity=3, maxLinearAcceleration=250, maxLinearVelocity=500, velocityScaleZ=0, worldAlpha=1}
   }
   n15 = Get Jump braid alpha() -> (Jump braid alpha:double)
   n16 = "Transform (Modify) Bone - Bone: L_braidend_master"(ComponentPose=@n14, Alpha=@n15, Translation="0,6,2", Rotation="0,0,0", Scale="1,1,1") {
    translationMode=BMM_Additive
    translationSpace=BCS_BoneSpace
    alphaBoolBlend {blendOption=Linear}
    alphaScaleBiasClamp {bInterpResult=true, interpSpeedIncreasing=2}
    boneToModify {boneName=L_braidend_master}
    translation {y=6, z=2}
   }
   n17 = "Transform (Modify) Bone - Bone: R_braidend_master"(ComponentPose=@n16, Alpha=@n15, Translation="0,-7,0", Rotation="0,0,0", Scale="1,1,1") {
    rotationMode=BMM_Additive
    rotationSpace=BCS_BoneSpace
    translationMode=BMM_Additive
    translationSpace=BCS_BoneSpace
    alphaBoolBlend {blendOption=Linear}
    alphaScaleBiasClamp {bInterpResult=true, interpSpeedIncreasing=2}
    boneToModify {boneName=R_braidend_master}
    translation {y=-7}
   }
   n18 = RigidBody(ComponentPose=@n17) {
    bClampLinearTranslationLimitToRefPose=true
    bOverrideWorldGravity=true
    overridePhysicsAsset=&/F_MED_Pink_Widow_Braids_FaceAcc_Earrings_Physics
    simulationSpace=BaseBoneSpace
    alphaBoolBlend {blendOption=Linear}
    baseBoneRef {boneName=head}
    componentAppliedLinearAccClamp=<3000 3000 3000>
    componentLinearAccScale=<1.2 1.2 1.2>
    componentLinearVelScale=<1.2 1.2 1.2>
    overrideWorldGravity {z=-980}
    simSpaceSettings {maxAngularAcceleration=1000, maxAngularVelocity=3, maxLinearAcceleration=250, maxLinearVelocity=500, velocityScaleZ=0, worldAlpha=1}
   }
   n19 = Get bIsInFrontEnd()
   n20 = "Transform (Modify) Bone - Bone: R_braidend_master"!(ComponentPose=@n18, bAlphaBoolEnabled:bool=@n19, Translation="0,-2,0", Rotation="0,0,0", Scale="1,1,1") -> (Pose) {
    alphaInputType=Bool
    translationMode=BMM_Additive
    translationSpace=BCS_BoneSpace
    alphaBoolBlend {blendOption=Linear}
    boneToModify {boneName=R_braidend_master}
    translation {y=-2}
   }
   n21 = "Transform (Modify) Bone - Bone: L_braidend_master"!(ComponentPose=@n20, bAlphaBoolEnabled:bool=@n19, Translation="-1,1,0", Rotation="0,0,0", Scale="1,1,1") -> (Pose) {
    alphaInputType=Bool
    translationMode=BMM_Additive
    translationSpace=BCS_BoneSpace
    alphaBoolBlend {blendOption=Linear}
    boneToModify {boneName=L_braidend_master}
    translation=<-1 1>
   }
   n22 = Control Rig(Source=Component To Local(ComponentPose=@n21)) {
    alphaBoolBlend {blendOption=Linear}
    controlRigAssetReference {blueprintRigClass=&/F_MED_PinkWidow_Braids_SecondaryCtrlRig.F_MED_PinkWidow_Braids_SecondaryCtrlRig_C}
   }
   Output Pose(Result=@n22) {layerGroup=None}
  ]}
 EventGraph {nodes [
   Event BlueprintUpdateAnimation()
   n1 = OR Boolean!(A=Get bIsSkyDiving(), B=Get bIsSkydiveFloating(), C:bool) -> (ReturnValue)
   > Set Braid alpha(Braid alpha:double=SelectFloat(A=0.2, B=SelectFloat(A=0.75, B=1, bPickA=@n1), bPickA=Get bIsDBNO())) -> (Output_Get:double)
   n2 = OR Boolean!(A=Get bIsJumping(), B=Get bIsFalling()) -> (ReturnValue)
   n3 = Get bIsDBNO()
   n4 = Get bIsSkyDiving()
   n5 = Get bIsSkydiveFloating()
   > Set Jump braid alpha(Jump braid alpha:double=SelectFloat(A=0, B=SelectFloat(A=1, B=0, bPickA=@n2), bPickA=OR Boolean!(A=@n4, B=@n3, C:bool=@n5) -> (ReturnValue))) -> (Output_Get:double)
  ]}
]
variables [
 Braid alpha {type=number}
 Jump braid alpha {type=number}
]
