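# Control Rig for the Pink Widow ponytail face accessory that offsets and rotates the ponytail's master joint as the head bends.
# It reads the head's rotation relative to spine_05 and remaps a negative range of Euler Z to a clamped weight.
# A delta-time accumulated lerp smooths that weight, which blends the fixed offset on master_ponytail_jnt.
# The dyn_ hair and earring joints are in the hierarchy but untouched by the graph.
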ControlRigBlueprint /BRCosmetics/Characters/Player/Female/Medium/Bodies/F_MED_Pink_Widow/Meshes/Parts/F_MED_Pink_Widow_Ponytail_FaceAcc_CtrlRig
graphs [
 RigVMFunctionLibrary {links [], nodes []}
 RigVMModel {nodes [
   RigUnit_BeginExecution = &RigUnit_BeginExecution()
   n1 = &RigVMFunction_MathQuaternionToEuler(
    RotationOrder=XZY
    Value=&RigUnit_GetRelativeTransformForItem(Child=(Type=Bone,Name="head"), Parent=(Type=Bone,Name="spine_05")).RelativeTransform.Rotation
   )
   &RigUnit_ModifyTransforms(
    ItemToModify=((Item=(Type=Bone,Name="master_ponytail_jnt"),Transform=(Rotation=(X=0,Y=0,Z=0,W=1),Translation=(X=-1,Y=0,Z=2),Scale3D=(X=1,Y=1,Z=1))))
    Weight=/Script/RigVM.RigVMFunction_AccumulateFloatLerp(
     Blend=10
     InitialValue=0
     bIntegrateDeltaTime=true
     TargetValue=&RigVMFunction_MathFloatRemap(SourceMaximum=-35, SourceMinimum=-10, TargetMinimum=0, bClamp=true, Value=@n1.Result.Z).Result
    ) {is=struct RigVMUnitNode, template="AccumulateLerp::Execute(in TargetValue,in InitialValue,in Blend,in bIntegrateDeltaTime,out Result)"}.Result
    ItemToModify.0.Transform.Rotation=&RigVMFunction_MathQuaternionFromEuler(Euler=(X=0,Y=-45,Z=0)).Result
    ExecutePin=@RigUnit_BeginExecution.ExecutePin
   )
  ]}
]
hierarchy [
 root {loc=<0 0 0>, rot=<0 0 0 1>}
 pelvis {parent=root, loc=<0 -0.5301 98.0571>, rot=<0.0127 0.707 -0.0127 -0.707>}
 spine_01 {parent=pelvis, loc=<6.1208 -0.9328 0>, rot=<0 0 0.0576 -0.9983>}
 spine_02 {parent=spine_01, loc=<6.6022 -0.0153 0>, rot=<0 0 0.0022 -1>}
 spine_03 {parent=spine_02, loc=<6.4919 -0.0105 0>, rot=<0 0 0.0015 -1>}
 spine_04 {parent=spine_03, loc=<6.9596 -0.007 0>, rot=<0 0 0.0009 -1>}
 spine_05 {parent=spine_04, loc=<6.9981 -0.0017 0>, rot=<0 0 -0.0042 -1>}
 clavicle_l {parent=spine_05, loc=<5.4831 -0.1053 -3.2946>, rot=<0.0938 -0.7388 0.0307 -0.6666>}
 neck_01 {parent=spine_05, loc=<11.3138 0.1529 0>, rot=<0 0 -0.2003 -0.9797>}
 neck_02 {parent=neck_01, loc=<4.5627 0 0>, rot=<0 0 0.0504 -0.9987>}
 head {parent=neck_02, loc=<5.3914 0 0>, rot=<0 0 0.115 -0.9934>}
 L_earring_master {parent=head, loc=<-0.3345 0.4056 -5.9972>, rot=<0.0041 -1 0 0>}
 dyn_L_earring_jnt {parent=L_earring_master, loc=<0.6424 0 0>, rot=<0 0 0 1>}
 R_earring_master {parent=head, loc=<-0.3346 0.4056 5.9972>, rot=<1 0.0041 0 0>}
 dyn_R_earring_jnt {parent=R_earring_master, loc=<-0.642 0 0>, rot=<0 0 0 1>}
 master_hair_jnt {parent=head, loc=<14.3055 0.386 0>, rot=<0.0029 -0.7071 -0.0029 -0.7071>}
 master_ponytail_jnt {parent=master_hair_jnt, loc=<0.005 -7.0203 -16.5348>, rot=<-0.3514 -0.3482 0.6173 -0.6117>}
 dyn_ponytail_1_jnt {parent=master_ponytail_jnt, loc=<1.0139 0 0.049>, rot=<0 -0.0455 0 -0.999>}
 dyn_ponytail_2_jnt {parent=dyn_ponytail_1_jnt, loc=<5.9178 -0.278 0.7477>, rot=<0 -0.0261 0 -0.9997>}
 dyn_ponytail_3_jnt {parent=dyn_ponytail_2_jnt, loc=<5.9894 0.3207 -1.2469>, rot=<0.004 -0.0601 -0.0032 -0.9982>}
 master_bangs_jnt {parent=master_hair_jnt, loc=<0.2755 8.8982 -3.2162>, rot=<-0.0892 -0.7062 -0.088 -0.6969>}
 dyn_R_sidefringe_piece_jnt {parent=master_bangs_jnt, loc=<4.4087 -12.7786 -7.6744>, rot=<-0.5936 0.4008 0.3983 -0.573>}
 dyn_bangs_jnt {parent=master_bangs_jnt, loc=<1.4662 0.1186 0>, rot=<0 0 0.1253 -0.9921>}
 dyn_R_sidefringe_jnt {parent=master_bangs_jnt, loc=<2.6708 -6.7449 -8.3033>, rot=<-0.4181 0.575 0.5798 -0.398>}
 dyn_L_sidefringe_jnt {parent=master_bangs_jnt, loc=<2.6328 -10.2455 7.9345>, rot=<-0.0031 0.0059 0.6686 -0.7436>}
 clavicle_r {parent=spine_05, loc=<5.4831 -0.1053 3.2946>, rot=<0.7388 0.0938 -0.6666 -0.0307>}
]
