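# Physics asset for the Pink Widow female ponytail and face accessory mesh, defining secondary motion for hair and earrings.
# Capsules on spine, clavicles, neck, head and master joints are linked by free angular constraints as the base for the dynamic joints.
# A three-link ponytail chain, bangs, side fringe pieces and both earrings hang from master joints with limited swing and very tight twist.
# Bangs are stiffest, while earrings swing widely on one axis only.
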PhysicsAsset /BRCosmetics/Characters/Player/Female/Medium/Bodies/F_MED_Pink_Widow/Meshes/Parts/F_MED_Pink_Widow_Ponytail_FaceAcc_Physics
bodies {
 L_earring_master [{shapeType=Capsule, length=1, radius=1, shapeName=L_earring_master_capsule}]
 R_earring_master [{shapeType=Capsule, length=1, radius=1, shapeName=R_earring_master_capsule}]
 clavicle_l [{shapeType=Capsule, length=5, radius=5, shapeName=clavicle_l_capsule, center=<7.1743 0 2.6112>, rotation=<rot -0.0003 -90 -109.9999>}]
 clavicle_r [{shapeType=Capsule, length=5, radius=5, shapeName=clavicle_r_capsule, center=<-6.9806 0 -2.5407>, rotation=<rot -0.0003 -90 -109.9999>}]
 dyn_L_earring_jnt [{
   shapeType=Capsule
   length=3.5129
   radius=1.6899
   shapeName=dyn_L_earring_jnt_capsule
   center=<3.4483 0.1352 0.0186>
   rotation=<rot 0 92.0613 -89.5712>
  }]
 dyn_L_sidefringe_jnt [{
   shapeType=Capsule
   length=4.7055
   radius=1.4914
   shapeName=dyn_L_sidefringe_jnt_capsule
   center=<3.0762 0.2692 -0.5571>
   rotation=<rot 0 -82.3488 -76.7708>
  }]
 dyn_R_earring_jnt [{
   shapeType=Capsule
   length=3.5129
   radius=1.6899
   shapeName=dyn_R_earring_jnt_capsule
   center=<-3.4485 -0.1352 -0.0186>
   rotation=<rot 0 92.0613 -89.5712>
  }]
 dyn_R_sidefringe_jnt [{
   shapeType=Capsule
   length=3.8778
   radius=2.3162
   shapeName=dyn_R_sidefringe_jnt_capsule
   center=<3.6578 -0.2569 -0.9127>
   rotation=<rot 0 -95.552 -83.8815>
  }]
 dyn_R_sidefringe_piece_jnt [{
   shapeType=Capsule
   length=2.4168
   radius=1.1155
   shapeName=dyn_R_sidefringe_piece_jnt_capsule
   center=<1.9251 0.1473 -0.7244>
   rotation=<rot 0 -82.9228 -75.3069>
  }]
 dyn_bangs_jnt [{
   shapeType=Capsule
   length=5.9758
   radius=2.8629
   shapeName=dyn_bangs_jnt_capsule
   center=<1.9506 2.1827 0.1607>
   rotation=<rot -1.1906 -90 -2.684>
  }]
 dyn_ponytail_1_jnt [{shapeType=Capsule, length=3.5, radius=2.5, shapeName=dyn_ponytail_1_jnt_capsule, center=<2.9795 0 0.4613>, rotation=<rot 0 0 90>}]
 dyn_ponytail_2_jnt [{shapeType=Capsule, length=2.25, radius=3.75, shapeName=dyn_ponytail_2_jnt_capsule, center=<3.1369 0 -0.4382>, rotation=<rot 0 0 90>}]
 dyn_ponytail_3_jnt [{shapeType=Capsule, length=4.25, radius=3.5, shapeName=dyn_ponytail_3_jnt_capsule, center=<5.3301 0 0>, rotation=<rot -90 0.0099 -0.0099>}]
 head [{shapeType=Capsule, length=5, radius=5, shapeName=head_capsule, center=<0 3.1491 0>, rotation=<rot -90 180 179.9997>}]
 master_bangs_jnt [{shapeType=Capsule, length=1, radius=1, shapeName=master_bangs_jnt_capsule}]
 master_hair_jnt [{shapeType=Capsule, length=1, radius=1, shapeName=master_hair_jnt_capsule}]
 master_ponytail_jnt [{shapeType=Capsule, length=1, radius=1, shapeName=master_ponytail_jnt_capsule}]
 neck_01 [{shapeType=Capsule, length=3, radius=3, shapeName=neck_01_capsule, rotation=<rot -90 -178.7342 178.734>}]
 neck_02 [{shapeType=Capsule, length=3, radius=3, shapeName=neck_02_capsule, rotation=<rot -90 -90 89.9998>}]
 spine_05 [{shapeType=Capsule, length=8, radius=8, shapeName=spine_05_capsule, center=<1.4709 1.1718 0>, rotation=<rot 0 -90 0>}]
}
constraints [
 {
  bone1Name=neck_02
  bone2Name=neck_01
  swing1LimitDegrees=45
  swing1Motion=Free
  swing2LimitDegrees=45
  swing2Motion=Free
  twistLimitDegrees=45
  twistMotion=Free
 }
 {
  bone1Name=head
  bone2Name=neck_02
  swing1LimitDegrees=45
  swing1Motion=Free
  swing2LimitDegrees=45
  swing2Motion=Free
  twistLimitDegrees=45
  twistMotion=Free
 }
 {
  bone1Name=L_earring_master
  bone2Name=head
  swing1LimitDegrees=45
  swing1Motion=Free
  swing2LimitDegrees=45
  swing2Motion=Free
  twistLimitDegrees=45
  twistMotion=Free
 }
 {bone1Name=dyn_L_earring_jnt, bone2Name=L_earring_master, swing1LimitDegrees=45, swing2LimitDegrees=2, twistLimitDegrees=2}
 {
  bone1Name=R_earring_master
  bone2Name=head
  swing1LimitDegrees=45
  swing1Motion=Free
  swing2LimitDegrees=45
  swing2Motion=Free
  twistLimitDegrees=45
  twistMotion=Free
 }
 {bone1Name=dyn_R_earring_jnt, bone2Name=R_earring_master, swing1LimitDegrees=45, swing2LimitDegrees=2, twistLimitDegrees=2}
 {
  bone1Name=master_hair_jnt
  bone2Name=head
  swing1LimitDegrees=45
  swing1Motion=Free
  swing2LimitDegrees=45
  swing2Motion=Free
  twistLimitDegrees=45
  twistMotion=Free
 }
 {
  bone1Name=master_ponytail_jnt
  bone2Name=master_hair_jnt
  swing1LimitDegrees=45
  swing1Motion=Free
  swing2LimitDegrees=45
  swing2Motion=Free
  twistLimitDegrees=45
  twistMotion=Free
 }
 {bone1Name=dyn_ponytail_1_jnt, bone2Name=master_ponytail_jnt, swing1LimitDegrees=45, swing2LimitDegrees=45, twistLimitDegrees=2}
 {bone1Name=dyn_ponytail_2_jnt, bone2Name=dyn_ponytail_1_jnt, swing1LimitDegrees=45, swing2LimitDegrees=45, twistLimitDegrees=2}
 {bone1Name=dyn_ponytail_3_jnt, bone2Name=dyn_ponytail_2_jnt, swing1LimitDegrees=45, swing2LimitDegrees=45, twistLimitDegrees=2}
 {
  bone1Name=master_bangs_jnt
  bone2Name=master_hair_jnt
  swing1LimitDegrees=45
  swing1Motion=Free
  swing2LimitDegrees=45
  swing2Motion=Free
  twistLimitDegrees=45
  twistMotion=Free
 }
 {bone1Name=dyn_R_sidefringe_piece_jnt, bone2Name=master_bangs_jnt, swing1LimitDegrees=35, swing2LimitDegrees=38, twistLimitDegrees=2}
 {bone1Name=dyn_bangs_jnt, bone2Name=master_bangs_jnt, swing1LimitDegrees=5, swing2LimitDegrees=5, twistLimitDegrees=1}
 {bone1Name=dyn_R_sidefringe_jnt, bone2Name=master_bangs_jnt, swing1LimitDegrees=15, swing2LimitDegrees=15, twistLimitDegrees=2}
 {bone1Name=dyn_L_sidefringe_jnt, bone2Name=master_bangs_jnt, swing1LimitDegrees=33, swing2LimitDegrees=33, twistLimitDegrees=2}
 {
  bone1Name=neck_01
  bone2Name=spine_05
  swing1LimitDegrees=45
  swing1Motion=Free
  swing2LimitDegrees=45
  swing2Motion=Free
  twistLimitDegrees=45
  twistMotion=Free
 }
 {
  bone1Name=clavicle_l
  bone2Name=spine_05
  swing1LimitDegrees=45
  swing1Motion=Free
  swing2LimitDegrees=45
  swing2Motion=Free
  twistLimitDegrees=45
  twistMotion=Free
 }
 {
  bone1Name=clavicle_r
  bone2Name=spine_05
  swing1LimitDegrees=45
  swing1Motion=Free
  swing2LimitDegrees=45
  swing2Motion=Free
  twistLimitDegrees=45
  twistMotion=Free
 }
]
