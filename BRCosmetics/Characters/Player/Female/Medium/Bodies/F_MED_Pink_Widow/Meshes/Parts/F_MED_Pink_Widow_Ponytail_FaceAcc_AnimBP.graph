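# Animation Blueprint for the ponytail face accessory of the Pink Widow cosmetic, adding physics motion that adapts to worn back gear.
# A pose copied from the attached parent mesh runs through a Control Rig and is cached for two head-space RigidBody simulations.
# The Pack physics asset is used with a backbling or cape and the standard one otherwise.
# Backbling and cape flags refresh on animation update, with backbling also updating on customization load.
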AnimBlueprint /BRCosmetics/Characters/Player/Female/Medium/Bodies/F_MED_Pink_Widow/Meshes/Parts/F_MED_Pink_Widow_Ponytail_FaceAcc_AnimBP
target_skeleton=&/F_MED_Pink_Widow_Ponytail_FaceAcc_Skeleton
graphs [
 AnimGraph {nodes [
   n1 = RigidBody(ComponentPose=Local To Component(LocalPose=Use cached pose 'FaceAcc_Cache'())) {
    bClampLinearTranslationLimitToRefPose=true
    bOverrideWorldGravity=true
    overridePhysicsAsset=&/F_MED_Pink_Widow_Ponytail_FaceAcc_Pack_Physics
    simulationSpace=BaseBoneSpace
    alphaBoolBlend {blendOption=Linear}
    baseBoneRef {boneName=head}
    componentAppliedLinearAccClamp=<3000 3000 3000>
    componentLinearAccScale=<1.2 1.2 1.2>
    componentLinearVelScale=<1.2 1.2 1.2>
    overrideWorldGravity {z=-980}
    simSpaceSettings {maxAngularAcceleration=1000, maxAngularVelocity=3, maxLinearAcceleration=250, maxLinearVelocity=500, velocityScaleZ=0, worldAlpha=1}
   }
   n2 = RigidBody(ComponentPose=Local To Component(LocalPose=Use cached pose 'FaceAcc_Cache'())) {
    bClampLinearTranslationLimitToRefPose=true
    bOverrideWorldGravity=true
    overridePhysicsAsset=&/F_MED_Pink_Widow_Ponytail_FaceAcc_Physics
    simulationSpace=BaseBoneSpace
    alphaBoolBlend {blendOption=Linear}
    baseBoneRef {boneName=head}
    componentAppliedLinearAccClamp=<3000 3000 3000>
    componentLinearAccScale=<0.9 0.9 0.9>
    componentLinearVelScale=<0.9 0.9 0.9>
    overrideWorldGravity {z=-980}
    simSpaceSettings {maxAngularAcceleration=1000, maxAngularVelocity=3, maxLinearAcceleration=250, maxLinearVelocity=500, velocityScaleZ=0, worldAlpha=1}
   }
   n3 = Get Is Wearing Backbling()
   n4 = Get Is Wearing Cape() -> (Is Wearing Cape:bool)
   Output Pose(Result=Blend Poses by bool(BlendPose_0=Component To Local(ComponentPose=@n1), BlendPose_1=Component To Local(ComponentPose=@n2), bActiveValue=OR Boolean(A=@n3, B=@n4)) {blendType=Linear}) {layerGroup=None}
   n5 = Control Rig(Source=Copy Pose From Mesh() {bUseAttachedParent=true}) {
    alphaBoolBlend {blendOption=Linear}
    controlRigAssetReference {blueprintRigClass=&/F_MED_Pink_Widow_Ponytail_FaceAcc_CtrlRig.F_MED_Pink_Widow_Ponytail_FaceAcc_CtrlRig_C}
   }
   Save cached pose 'FaceAcc_Cache'(Pose=@n5) {cacheName=FaceAcc_Cache}
  ]}
 EventGraph {nodes [
   n1 = Event BlueprintUpdateAnimation()
   n2 = CheckForBackbling()
   Set Is Wearing Backbling(execute=[@n1.then, @n3.then], Is Wearing Backbling=OR Boolean(A=@n2, B=@n2.bIsBackblingLong))
   > Set Is Wearing Cape(Is Wearing Cape=@n2.bIsBackblingACape)
   n3 = Event OnCharacterCustomizationLoaded()
  ]}
]
variables [
 Is Wearing Backbling {type=boolean}
 Is Wearing Cape {type=boolean}
]
